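# Niagara swing trail for the Sahara (male) pickaxe, drawing lightning-styled ribbons tinted by User.Color, light blue by default.
# Two near-identical emitters spawn anim trails from skeletal mesh bone or socket indices, with per-spawn-group lifetimes and height offsets on short, randomized loops.
# Curl noise distorts the ribbons while curves fade alpha and width and RGB is boosted.
# The second emitter is dimmer, longer-lived and adds a very low-alpha color attribute.
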NiagaraSystem /BRCosmetics/Weapons/FORT_Melee/Pickaxe_Sahara_Male/FX/NS_Pickaxe_Sahara_Trail
dependencies {
 usedDataInterfaces [&NiagaraDataInterfaceArrayFloat, &NiagaraDataInterfaceCurve]
 usedDynamicInputs []
 usedModules [&SystemState]
}
emitters {
 Lightning_01 {
  emitterUpdateScript [
   &EmitterState(Life Cycle Mode=::NewEnumerator1, Loop Behavior=::NewEnumerator1, Loop Duration=0.25, UseLoopDelay=false)
   &FN_SpawnAnimTrail_v2(Number of Anim Trails=2)
   /SetVariables(Emitter.LoopRandomFloat:NiagaraFloat=&RandomRangeFloat(Minimum=0.15, Maximum=0.22)) {isSetParameters=true, module=SetVariables_#1}
  ]
  particleSpawnScript [
   /Game/Effects/Niagara/Modules/AnimTrail/FN_InitializeAnimTrail/FN_InitializeAnimTrailBoneIndecies(First Bone or Socket Index=2, Second Bone or Socket Index=1)
   &FN_InitializeAnimTrailBoneIndecies(AnimTrail Selection=1, First Bone or Socket Index=2, Second Bone or Socket Index=1) {module=FN_InitializeAnimTrailBoneIndecies001}
   /Game/Effects/Niagara/Modules/AnimTrail/FN_InitializeAnimTrail/FN_InitializeAnimTrail_v3/FN_InitializeAnimTrail(
    AnimTrailWidth=1.5
    Color:LinearColor=&MultiplyLinearColor(Linear Color=$User.Color, Scale Factor=&MakeVector4_FromFloat(Float=1))
    Lifetime:NiagaraFloat=&Multiply_Float(
     A=$Emitter.LoopRandomFloat
     B=&SelectFloatFromArray(Direct Array Index=$Particles.SpawnGroup, Float Selection Array=DataInterface() {propertyValues {FloatData [0.75,0.5]}})
    )
    Position Offset:Vector3f=&MaskVectorBySpawnGroup(
     Masked Vector=&RandomRangeVector(Minimum=<0 0 10>, Maximum=<0 0 12>, Randomness Mode=::NewEnumerator1, Random Seed=12, Fixed Random Seed=true)
     Passthrough Vector=&RandomRangeVector(Minimum=<0 0 0>, Maximum=<0 0 3>, Randomness Mode=::NewEnumerator1, Random Seed=12, Fixed Random Seed=true)
    )
   ) {module=FN_InitializeAnimTrail001}
  ]
  particleUpdateScript [
   &CurlNoiseForce(Noise Frequency=10, Noise Strength=&RandomRangeFloat(Minimum=250, Maximum=500), Pan Noise=false)
   &SolveForcesAndVelocity(Clamp Velocity=false, Limit Acceleration=false)
   /Game/Effects/Niagara/Modules/AnimTrail/FN_UpdateAnimTrail/FN_UpdateAnimTrail_v3/FN_UpdateAnimTrail(Locked Particle RibbonWidth:NiagaraFloat=$Particles.RibbonWidth)
   &ScaleColor(
    Scale Alpha=&FloatFromCurve(
     FloatCurve=DataInterface() {propertyValues {Curve {keys [
         {arriveTangent=-0.102, interpMode=RCIM_Cubic, leaveTangent=-0.102, tangentMode=RCTM_User, time=0.5, value=1}
         {arriveTangent=-2.7763, interpMode=RCIM_Cubic, leaveTangent=-2.7763, tangentMode=RCTM_User, time=1, value=0}
        ]}}}
     Scale Curve=2
    )
    Scale RGB=&VectorFromFloat(Value=5)
    ScaleA=true
    ScaleRGB=true
   )
   &ScaleRibbonWidth(Ribbon Width Scale=&FloatFromCurve(FloatCurve=DataInterface() {propertyValues {Curve {keys [{interpMode=RCIM_Cubic, tangentMode=RCTM_User, time=0.5, value=1}, {interpMode=RCIM_Cubic, tangentMode=RCTM_User, time=1, value=0}]}}}))
  ]
  rendererClasses [&NiagaraRibbonRendererProperties]
  renderers [{rendererClass=&NiagaraRibbonRendererProperties}]
 }
 Lightning_01001 {
  emitterUpdateScript [
   &EmitterState(Life Cycle Mode=::NewEnumerator1, Loop Behavior=::NewEnumerator1, Loop Duration=0.25, UseLoopDelay=false)
   &FN_SpawnAnimTrail_v2(Number of Anim Trails=2)
   /SetVariables(Emitter.LoopRandomFloat:NiagaraFloat=&RandomRangeFloat(Minimum=0.15, Maximum=0.22)) {isSetParameters=true, module=SetVariables_#1}
  ]
  particleSpawnScript [
   &FN_InitializeAnimTrailBoneIndecies(First Bone or Socket Index=2, Second Bone or Socket Index=1)
   &FN_InitializeAnimTrailBoneIndecies(AnimTrail Selection=1, First Bone or Socket Index=2, Second Bone or Socket Index=1) {module=FN_InitializeAnimTrailBoneIndecies001}
   &FN_InitializeAnimTrail(
    AnimTrailWidth=1.5
    Color:LinearColor=&MultiplyLinearColor(Linear Color=$User.Color, Scale Factor=&MakeVector4_FromFloat(Float=0.75))
    Lifetime:NiagaraFloat=&Multiply_Float(
     A=$Emitter.LoopRandomFloat
     B=&SelectFloatFromArray(Direct Array Index=$Particles.SpawnGroup, Float Selection Array=DataInterface() {propertyValues {FloatData [1,0.75]}})
    )
    Position Offset:Vector3f=&MaskVectorBySpawnGroup(
     Masked Vector=&RandomRangeVector(Minimum=<0 0 10>, Maximum=<0 0 12>, Randomness Mode=::NewEnumerator1, Random Seed=12)
     Passthrough Vector=&RandomRangeVector(Minimum=<0 0 0>, Maximum=<0 0 3>, Randomness Mode=::NewEnumerator1, Random Seed=12)
    )
   ) {module=FN_InitializeAnimTrail001}
  ]
  particleUpdateScript [
   &CurlNoiseForce(Noise Frequency=10, Noise Strength=&RandomRangeFloat(Minimum=250, Maximum=500), Pan Noise=false)
   &SolveForcesAndVelocity(Clamp Velocity=false, Limit Acceleration=false)
   &FN_UpdateAnimTrail(Locked Particle RibbonWidth:NiagaraFloat=$Particles.RibbonWidth)
   &ScaleColor(
    Scale Alpha=&Multiply_Float(
     A=&FloatFromCurve(
      FloatCurve=DataInterface() {propertyValues {Curve {keys [
          {arriveTangent=-0.102, interpMode=RCIM_Cubic, leaveTangent=-0.102, tangentMode=RCTM_User, time=0.5, value=1}
          {arriveTangent=-2.7763, interpMode=RCIM_Cubic, leaveTangent=-2.7763, tangentMode=RCTM_User, time=1, value=0}
         ]}}}
      Scale Curve=2
     )
     B=1
    )
    Scale RGB=&VectorFromFloat(Value=5)
    ScaleA=true
    ScaleRGB=true
   )
   &ScaleRibbonWidth(Ribbon Width Scale=&Multiply_Float(
     A=&FloatFromCurve(FloatCurve=DataInterface() {propertyValues {Curve {keys [{interpMode=RCIM_Cubic, tangentMode=RCTM_User, time=0.5, value=1}, {interpMode=RCIM_Cubic, tangentMode=RCTM_User, time=1, value=0}]}}})
     B=1
    ))
   &FN_Sprite_Attributes(
    Color=&MultiplyLinearColor(Linear Color=$Particles.Color, Scale Factor=<1 1 1 0.015>)
    Create Color Attribute=true
    Write to Module Namespace=true
   )
  ]
  rendererClasses [&NiagaraRibbonRendererProperties]
  renderers [{rendererClass=&NiagaraRibbonRendererProperties}]
 }
}
summary {
 systemName=NS_Pickaxe_Sahara_Trail
 emitters [
  {emitterName=Lightning_01, rendererClasses [&NiagaraRibbonRendererProperties]}
  {emitterName=Lightning_01001, rendererClasses [&NiagaraRibbonRendererProperties]}
 ]
 userVariables [
  User.Color {type=LinearColor, description="", defaultValue {struct=LinearColor, value=<rgba 0.25 0.6 1 1>}}
  User.RibbonIDOffset {type=NiagaraInt32}
  User.SkeletalMesh {
   type=NiagaraDataInterfaceSkeletalMesh
   defaultValue {
    struct=NiagaraExt_VariableValue_DataInterface
    value {dataInterface=&self:NiagaraDataInterfaceSkeletalMesh_0}
   }
  }
 ]
}
