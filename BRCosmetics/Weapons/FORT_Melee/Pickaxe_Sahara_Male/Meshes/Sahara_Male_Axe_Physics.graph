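# Physics asset for the Sahara Male pickaxe mesh, defining collision and joint limits for a dangling strap on the axe.
# A capsule body covers the two-handed handle, and three capsules of growing radius cover the dyn_strap bones.
# Constraints chain each strap segment to the previous one and the first to the handle, sharing identical swing and tight twist limits so the strap can sway.
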PhysicsAsset /BRCosmetics/Weapons/FORT_Melee/Pickaxe_Sahara_Male/Meshes/Sahara_Male_Axe_Physics
bodies {
 Melee_TwoHanded_Handle [{shapeType=Capsule, length=29.7381, radius=2.5308, center=<-0.1238 -0.2023 -2.6187>}]
 dyn_strap_1 [{shapeType=Capsule, length=0.75, radius=1.5, shapeName=dyn_strap_1_capsule, center=<2.7814 0 0>, rotation=<rot 90 0 0>}]
 dyn_strap_2 [{shapeType=Capsule, length=0.75, radius=3, shapeName=dyn_strap_2_capsule, center=<1.5402 0.0003 0>, rotation=<rot 90 0 0>}]
 dyn_strap_3 [{shapeType=Capsule, length=1.5, radius=3.5, shapeName=dyn_strap_3_capsule, center=<3.1246 0.0346 0>, rotation=<rot 90 0 0>}]
}
constraints [
 {bone1Name=dyn_strap_3, bone2Name=dyn_strap_2, swing1LimitDegrees=60, swing2LimitDegrees=40, twistLimitDegrees=15}
 {bone1Name=dyn_strap_2, bone2Name=dyn_strap_1, swing1LimitDegrees=60, swing2LimitDegrees=40, twistLimitDegrees=15}
 {bone1Name=dyn_strap_1, bone2Name=Melee_TwoHanded_Handle, swing1LimitDegrees=60, swing2LimitDegrees=40, twistLimitDegrees=15}
]
