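# Anim Blueprint for the Sahara Male pickaxe mesh that adds secondary physics motion to the weapon, including its strap bones.
# A RigidBody node using the axe physics asset and an AnimDynamics chain across three strap bones each process an input pose.
# A Blend Poses by bool joins them with its bool unwired, leaving the AnimDynamics branch selected by default.
# The event graph holds only unconnected stub nodes.
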AnimBlueprint /BRCosmetics/Weapons/FORT_Melee/Pickaxe_Sahara_Male/Meshes/Sahara_Male_Axe_AnimBP
target_skeleton=&/Sahara_Male_Axe_Skeleton
graphs [
 AnimGraph {nodes [
   AnimGraphNode_SubInput_0 = Input Pose()
   n1 = RigidBody(ComponentPose=Local To Component(LocalPose=@AnimGraphNode_SubInput_0)) {
    bOverrideWorldGravity=true
    overridePhysicsAsset=&/Sahara_Male_Axe_Physics
    alphaBoolBlend {blendOption=Linear}
    componentAppliedLinearAccClamp=<3000 3000 3000>
    componentLinearAccScale=<0.3 0.3 0.3>
    componentLinearVelScale=<0.3 0.3 0.3>
    overrideWorldGravity {z=-980}
   }
   n2 = "Anim Dynamics - Chain: dyn_strap_1 -> dyn_strap_3"(ComponentPose=Local To Component(LocalPose=Input Pose() {name=InPose_0}), Alpha=1, ExternalForce="0,0,0") {
    angularBiasOverride=0.5
    angularDampingOverride=0.9
    angularSpringConstant=5
    bAngularSpring=true
    bChain=true
    bOverrideAngularBias=true
    bOverrideAngularDamping=true
    gravityScale=0.5
    simulationSpace=Actor
    alphaBoolBlend {blendOption=Linear}
    boundBone {boneName=dyn_strap_1}
    chainEnd {boneName=dyn_strap_3}
    gravityOverride {z=-980}
    physicsBodyDefinitions [
     {
      boundBone {boneName=dyn_strap_1}
      boxExtents=<2 5 5>
      constraintSetup {angularLimitsMax=<5 30 30>, angularLimitsMin=<-5 -30 -30>, angularTarget=<0 0 0>}
      localJointOffset=<2 0 0>
     }
     {
      boundBone {boneName=dyn_strap_2}
      boxExtents=<2 5 5>
      constraintSetup {angularLimitsMax=<5 30 30>, angularLimitsMin=<-5 -30 -30>, angularTarget=<0 0 0>}
      localJointOffset=<2 0 0>
     }
     {
      boundBone {boneName=dyn_strap_3}
      boxExtents=<2 5 5>
      constraintSetup {angularLimitsMax=<5 30 30>, angularLimitsMin=<-5 -30 -30>, angularTarget=<0 0 0>}
      localJointOffset=<2 0 0>
     }
    ]
   }
   Output Pose(Result=Blend Poses by bool(BlendPose_0=Component To Local(ComponentPose=@n1), BlendPose_1=Component To Local(ComponentPose=@n2), BlendTime_0=0.2, BlendTime_1=0.2) {blendType=Linear}) {layerGroup=None}
  ]}
 EventGraph {nodes [
   TryGetPawnOwner()
   Event BlueprintUpdateAnimation()
  ]}
]
variables []
